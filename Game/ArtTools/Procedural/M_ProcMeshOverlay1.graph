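# M_ProcMeshOverlay1 is an additive material, named as a procedural mesh overlay, that builds a screen-aligned procedural pattern.
# Screen-aligned UVs are multiplied by Scale and camera-to-object distance, so pattern frequency follows distance, then rotated by the rot parameter.
# Sine, Sign and Saturate turn those coordinates into hard-edged bands.
# A dark red Color parameter, a camera-vector branch and a Divide stand unconnected, and no material output links appear.
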Material /Game/ArtTools/Procedural/M_ProcMeshOverlay1
graph {
 properties {blendMode=BLEND_Additive, pixelDepthOffsetMode=PDOM_Legacy, usedShadingModels=Default Lit}
 nodes [
  n1 = ScalarParameter() {defaultValue=0.06, parameterName=Scale}
  n2 = Distance(A=ObjectPositionWS(), B=CameraPositionWS())
  Divide(A=@n1, B=@n2)
  n3 = StaticBool() {value=true}
  Multiply(A=Normalize(VectorInput=MaterialFunctionCall!(Use World Position Offsets=@n3) {materialFunction=/Engine/Functions/Engine_MaterialFunctions02/Utility/CameraVectorWithWPOOptions}.Result)) {constB=10}
  n4 = Rotator(
   Coordinate=Multiply(
    A=MaterialFunctionCall() {materialFunction=&ScreenAlignedUVs}."X 100%, Y 100%"
    B=Multiply(A=@n1, B=@n2)
   )
   Time=ScalarParameter() {defaultValue=2.5099, parameterName=rot}
  )
  Saturate(None=Sign(None=Sine(None=@n4)))
  VectorParameter() {parameterName=Color, defaultValue=<rgba 0.1208 0 0 1>}
 ]
}
