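; Placeable wrapper actor for procedural tools that hosts a child actor class and keeps transform edits relative to its assigned base.
; Construction stores the actor's offset from its world base as EditedTransform, which UpdateTransform reapplies after rebuilding the base from a parent transform and local base.
; The EditorOnlyOverlay component copies the child's static mesh and applies M_ProcMeshOverlay to every slot.
; Convert To Actor spawns the child class at this transform and destroys the wrapper.
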(event EventBeginPlay)

(event Collision|EventActorBeginOverlap
  (params (OtherActor :type Actor%20Object%20Reference)))

(event EventTick
  (params (DeltaSeconds :type Float%20%28single-precision%29)))

(event Custom|SetChildActor
  (params (Class :type Actor%20Class%20Reference) (Relative%20Transform :type Transform) (World%20Transform :type Transform) (Tag :type Name))
  (ChildActorComponent|SetChildActorClass (Variables|Default|GetChildActor) Class)
  (Variables|Default|SetLocalBaseTransform Relative%20Transform)
  (Variables|Default|SetWorldBaseTransform World%20Transform)
  (CallFunction|SetEditTransform)
  (CallFunction|UpdateOverlayMesh 0))

(event Custom|UpdateTransform
  (params (TransformParent :type Transform))
  (Transformation|SetActorTransform TransformParent false true)
  (Transformation|AddActorLocalTransform (Variables|Default|GetLocalBaseTransform) false true)
  (Variables|Default|SetWorldBaseTransform (Transformation|GetActorTransform))
  (Transformation|AddActorLocalTransform (Variables|Default|GetEditedTransform)))

(event Custom|Convert%20To%20Actor
  (Game|SpawnActorfromClass (Variables|/Script/Engine.ChildActorComponent|GetChildActorClass (Variables|Default|GetChildActor)) (Transformation|GetActorTransform) "Undefined" "SelectDefaultAtRuntime" 0 0)
  (Actor|DestroyActor))

(fn OverrideEditTransform
  (params (Transform :type Transform))
  (Variables|Default|SetEditedTransform Transform)
  (Transformation|AddActorLocalTransform (Variables|Default|GetEditedTransform)))

(fn SetEditTransform
  (Variables|Default|SetEditedTransform (Math|Transform|MakeRelativeTransform (Transformation|GetActorTransform) (Math|Transform|MakeTransform (.location (Variables|Default|GetWorldBaseTransform)) (.rotation (Variables|Default|GetWorldBaseTransform)) (Math|Vector|MakeVector 1.0 1.0 1.0)))))

(fn UpdateOverlayMesh
  (params (NewParam :type Static%20Mesh%20Object%20Reference))
  (bind _childactor (Variables|Default|GetChildActor))
  (bind _childactor_2 (Variables|/Script/Engine.ChildActorComponent|GetChildActor _childactor))
  (bind _returnvalue (Actor|GetComponentbyClass _childactor_2 "/Script/Engine.StaticMeshComponent"))
  (bind _editoronlyoverlay (Variables|Default|GetEditorOnlyOverlay))
  (bind _staticmesh (Variables|/Script/Engine.StaticMeshComponent|GetStaticMesh _returnvalue))
  (Utilities|IsValid _childactor
    (:Is%20Valid
      (Utilities|IsValid _childactor_2
        (:Is%20Valid
          (Utilities|IsValid _staticmesh
            (:Is%20Valid
              (Components|StaticMesh|SetStaticMesh _editoronlyoverlay _staticmesh)
              (for (_ _array_index) (Rendering|Material|GetMaterials _editoronlyoverlay)
                (Rendering|Material|SetMaterial _editoronlyoverlay _array_index "/Game/ArtTools/Procedural/M_ProcMeshOverlay.M_ProcMeshOverlay")))))))))

(fn ConstructionScript
  (CallFunction|SetEditTransform)
  (CallFunction|UpdateOverlayMesh 0))
