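; Editor prop-scattering tool that places meshes and Blueprint actors from a Procedural_DataAsset on traced surfaces within a radius ring.
; Construction script splits the collection into up to six HISM components plus an actor list, then rescatters instances with seeded streams.
; Custom events spawn tagged Procedural_SupportActors, either restoring saved edit transforms or randomizing the seed.
; Other events delete the scatter or snap actors to the floor; a per-copy identifier tag tracks ownership.
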(event Custom|2%20DeleteScatterActors
  (CallFunction|DeleteScatteredActors)
  (CallFunction|DeleteScatteredInstances))

(event Custom|1b%20Scatter%20DetailAsset%20Random
  (bind _count (Variables|Default|GetCount))
  (bind _iteration (Variables|Default|GetIteration))
  (bind _randomseed (Variables|Procedural|GetRandomSeed))
  (bind _returnvalue (Math|Integer|int+int _randomseed _iteration _count))
  (bind _returnvalue_2 (Math|Random|MakeRandomStream _returnvalue))
  (bind _returnvalue_3 (Math|Random|RandomFloatinRangefromStream _returnvalue_2 0.0 3.6e+02))
  (bind _returnvalue_4 (Transformation|GetActorRotation))
  (Variables|Default|SetResetScatter true)
  (Variables|Procedural|SetRandomSeed (Math|Random|RandomInteger 2147483647))
  (CallFunction|DeleteScatteredActors)
  (CallFunction|DeleteScatteredInstances)
  (Utilities|FlowControl|Sequence
    (:then_0
      ;[1 1 1] Make Unique Identifier
      (if (== (Variables|DoNotEdit|GetUniqueIdentifier) 0)
        (Variables|DoNotEdit|SetUniqueIdentifier (Math|Random|RandomIntegerinRange 1 999999999)))
      (bind _outactors (Actor|GetAllActorsOfClass (Utilities|GetClass self)))
      (for _array_element _outactors
        (if (Utilities|Equal(Object) _array_element self)
          _
          (elif (== (Variables|DoNotEdit|GetUniqueIdentifier _array_element) (Variables|DoNotEdit|GetUniqueIdentifier))
            (Variables|DoNotEdit|SetUniqueIdentifier (Math|Random|RandomIntegerinRange 1 999999999))))))
    (:then_1
      ;[1 1 1] makeUnique Tag
      (Variables|DoNotEdit|SetsearchTag (Utilities|String|StringToName (Utilities|String|Append "PPA_" (Utilities|String|ToString(Integer) (Variables|DoNotEdit|GetUniqueIdentifier))))))
    (:then_2
      (CallFunction|ScatterInstanceMeshes))
    (:then_3
      (for (_array_element_2 _array_index) (Variables|Default|GetACTOR_List)
        (Variables|Default|SetCount _array_index)
        (Math|Random|SetRandomStreamSeed (Variables|Default|GetStream) (+ (Variables|Procedural|GetRandomSeed) _array_index))
        (for _index (range 1 (+ (.Count_5_7BCF801C487720A69910A9A310D929B0 (Utilities|Struct|BreakProceduralCollection _array_element_2)) 1))
          (Variables|Default|SetIteration _index)
          (bind (_outhit _returnvalue_5) (Collision|LineTraceByChannel :Start (+ (Math|Vector|MakeVector 0.0 0.0 (Variables|Procedural|GetHeight)) (+ (Transformation|GetActorLocation) (Math|Vector|RotateVector (* (* (Math|Vector|Normalize (* (Math|Random|RandomUnitVectorfromStream (Variables|Default|GetStream)) (Math|Vector|MakeVector 1.0 1.0))) (Math|Random|RandomFloatinRangefromStream (Variables|Default|GetStream) (Variables|Procedural|GetRadius_Min) (Variables|Procedural|GetRadius_Max))) (Transformation|GetActorScale3D)) (Transformation|GetActorRotation)))) :End (+ (+ (Transformation|GetActorLocation) (Math|Vector|RotateVector (* (* (Math|Vector|Normalize (* (Math|Random|RandomUnitVectorfromStream (Variables|Default|GetStream)) (Math|Vector|MakeVector 1.0 1.0))) (Math|Random|RandomFloatinRangefromStream (Variables|Default|GetStream) (Variables|Procedural|GetRadius_Min) (Variables|Procedural|GetRadius_Max))) (Transformation|GetActorScale3D)) (Transformation|GetActorRotation))) (Math|Vector|MakeVector 0.0 0.0 -3e+03)) :TraceChannel "TraceTypeQuery10" :bTraceComplex true :DrawDebugType "ForOneFrame" :bIgnoreSelf true :TraceColor "(R=1.000000,G=0.000000,B=0.000000,A=1.000000)" :TraceHitColor "(R=0.000000,G=1.000000,B=0.000000,A=1.000000)" :DrawTime 2.0))
          (if _returnvalue_5
            (bind _returnvalue_6 (Game|SpawnActorfromClass "/Game/ArtTools/Procedural/Procedural_SupportActor.Procedural_SupportActor_C" (Math|Transform|MakeTransform (.Location (Collision|BreakHitResult _outhit)) (Math|Rotator|CombineRotators _returnvalue_4 (Math|Rotator|MakeRotfromXZ (Math|Vector|RotateVectorAroundAxis (Math|Rotator|GetRotationXVector (Math|Rotator|MakeRotfromZ (.Normal (Collision|BreakHitResult _outhit)))) _returnvalue_3 (.Normal (Collision|BreakHitResult _outhit))) (.Normal (Collision|BreakHitResult _outhit)))) (Math|Vector|MakeVector (Math|Random|RandomFloatinRangefromStream _returnvalue_2 (.ScaleMin_8_FE7EBF17444458437DDE60BE2A001AF5 (Utilities|Struct|BreakProceduralCollection _array_element_2)) (.ScaleMax_10_1A61A5DA412FCD256B46628068A60EA3 (Utilities|Struct|BreakProceduralCollection _array_element_2))) (Math|Random|RandomFloatinRangefromStream _returnvalue_2 (.ScaleMin_8_FE7EBF17444458437DDE60BE2A001AF5 (Utilities|Struct|BreakProceduralCollection _array_element_2)) (.ScaleMax_10_1A61A5DA412FCD256B46628068A60EA3 (Utilities|Struct|BreakProceduralCollection _array_element_2))) (Math|Random|RandomFloatinRangefromStream _returnvalue_2 (.ScaleMin_8_FE7EBF17444458437DDE60BE2A001AF5 (Utilities|Struct|BreakProceduralCollection _array_element_2)) (.ScaleMax_10_1A61A5DA412FCD256B46628068A60EA3 (Utilities|Struct|BreakProceduralCollection _array_element_2))))) "Undefined" "OverrideRootScale" 0 0))
            (Class|ProceduralSupportActor|SetChildActor _returnvalue_6 (.Blueprint_24_336B25D4459493F1D17B81BA5C9C10CA (Utilities|Struct|BreakProceduralCollection _array_element_2)) (Math|Transform|MakeRelativeTransform (Math|Transform|MakeTransform (.Location (Collision|BreakHitResult _outhit)) (Math|Rotator|CombineRotators _returnvalue_4 (Math|Rotator|MakeRotfromXZ (Math|Vector|RotateVectorAroundAxis (Math|Rotator|GetRotationXVector (Math|Rotator|MakeRotfromZ (.Normal (Collision|BreakHitResult _outhit)))) _returnvalue_3 (.Normal (Collision|BreakHitResult _outhit))) (.Normal (Collision|BreakHitResult _outhit)))) (Math|Vector|MakeVector (Math|Random|RandomFloatinRangefromStream _returnvalue_2 (.ScaleMin_8_FE7EBF17444458437DDE60BE2A001AF5 (Utilities|Struct|BreakProceduralCollection _array_element_2)) (.ScaleMax_10_1A61A5DA412FCD256B46628068A60EA3 (Utilities|Struct|BreakProceduralCollection _array_element_2))) (Math|Random|RandomFloatinRangefromStream _returnvalue_2 (.ScaleMin_8_FE7EBF17444458437DDE60BE2A001AF5 (Utilities|Struct|BreakProceduralCollection _array_element_2)) (.ScaleMax_10_1A61A5DA412FCD256B46628068A60EA3 (Utilities|Struct|BreakProceduralCollection _array_element_2))) (Math|Random|RandomFloatinRangefromStream _returnvalue_2 (.ScaleMin_8_FE7EBF17444458437DDE60BE2A001AF5 (Utilities|Struct|BreakProceduralCollection _array_element_2)) (.ScaleMax_10_1A61A5DA412FCD256B46628068A60EA3 (Utilities|Struct|BreakProceduralCollection _array_element_2))))) (Transformation|GetActorTransform)) (Math|Transform|MakeTransform (.Location (Collision|BreakHitResult _outhit)) (Math|Rotator|CombineRotators _returnvalue_4 (Math|Rotator|MakeRotfromXZ (Math|Vector|RotateVectorAroundAxis (Math|Rotator|GetRotationXVector (Math|Rotator|MakeRotfromZ (.Normal (Collision|BreakHitResult _outhit)))) _returnvalue_3 (.Normal (Collision|BreakHitResult _outhit))) (.Normal (Collision|BreakHitResult _outhit)))) (Math|Vector|MakeVector (Math|Random|RandomFloatinRangefromStream _returnvalue_2 (.ScaleMin_8_FE7EBF17444458437DDE60BE2A001AF5 (Utilities|Struct|BreakProceduralCollection _array_element_2)) (.ScaleMax_10_1A61A5DA412FCD256B46628068A60EA3 (Utilities|Struct|BreakProceduralCollection _array_element_2))) (Math|Random|RandomFloatinRangefromStream _returnvalue_2 (.ScaleMin_8_FE7EBF17444458437DDE60BE2A001AF5 (Utilities|Struct|BreakProceduralCollection _array_element_2)) (.ScaleMax_10_1A61A5DA412FCD256B46628068A60EA3 (Utilities|Struct|BreakProceduralCollection _array_element_2))) (Math|Random|RandomFloatinRangefromStream _returnvalue_2 (.ScaleMin_8_FE7EBF17444458437DDE60BE2A001AF5 (Utilities|Struct|BreakProceduralCollection _array_element_2)) (.ScaleMax_10_1A61A5DA412FCD256B46628068A60EA3 (Utilities|Struct|BreakProceduralCollection _array_element_2))))) "None" :!class /Game/ArtTools/Procedural/Procedural_SupportActor.Procedural_SupportActor_C)
            (if (not (Variables|Default|GetResetScatter))
              (Variables|Default|SetCurrentEditLocation (Math|Transform|MakeTransform "0, 0, 0"))
              (bind _array_element_3 (Utilities|Array|ForEachLoopwithBreak (Variables|Default|GetActorEditLocationsArray))
                (:LoopBody
                  (if (== (Utilities|Struct|BreakProceduralEditLocation _array_element_3) (Variables|Default|GetCount))
                    (if (== (.Iteration_4_6A5FC8554676355A6CD0528E37AA79EB (Utilities|Struct|BreakProceduralEditLocation _array_element_3)) (- (Variables|Default|GetIteration) 1))
                      (Variables|Default|SetCurrentEditLocation (.EditTransform_9_49D70F1F48C2C776CD90B5BC2FE0BEAE (Utilities|Struct|BreakProceduralEditLocation _array_element_3)))
                      (Variables|Default|SetResetScatter false))))
                (:Completed
                  _))
              (Class|ProceduralSupportActor|OverrideEditTransform _returnvalue_6 (Math|Transform|MakeRelativeTransform (Variables|Default|GetCurrentEditLocation) (Math|Transform|MakeTransform (.location (Math|Transform|MakeTransform (.Location (Collision|BreakHitResult _outhit)) (Math|Rotator|CombineRotators _returnvalue_4 (Math|Rotator|MakeRotfromXZ (Math|Vector|RotateVectorAroundAxis (Math|Rotator|GetRotationXVector (Math|Rotator|MakeRotfromZ (.Normal (Collision|BreakHitResult _outhit)))) _returnvalue_3 (.Normal (Collision|BreakHitResult _outhit))) (.Normal (Collision|BreakHitResult _outhit)))) (Math|Vector|MakeVector (Math|Random|RandomFloatinRangefromStream _returnvalue_2 (.ScaleMin_8_FE7EBF17444458437DDE60BE2A001AF5 (Utilities|Struct|BreakProceduralCollection _array_element_2)) (.ScaleMax_10_1A61A5DA412FCD256B46628068A60EA3 (Utilities|Struct|BreakProceduralCollection _array_element_2))) (Math|Random|RandomFloatinRangefromStream _returnvalue_2 (.ScaleMin_8_FE7EBF17444458437DDE60BE2A001AF5 (Utilities|Struct|BreakProceduralCollection _array_element_2)) (.ScaleMax_10_1A61A5DA412FCD256B46628068A60EA3 (Utilities|Struct|BreakProceduralCollection _array_element_2))) (Math|Random|RandomFloatinRangefromStream _returnvalue_2 (.ScaleMin_8_FE7EBF17444458437DDE60BE2A001AF5 (Utilities|Struct|BreakProceduralCollection _array_element_2)) (.ScaleMax_10_1A61A5DA412FCD256B46628068A60EA3 (Utilities|Struct|BreakProceduralCollection _array_element_2)))))) (.rotation (Math|Transform|MakeTransform (.Location (Collision|BreakHitResult _outhit)) (Math|Rotator|CombineRotators _returnvalue_4 (Math|Rotator|MakeRotfromXZ (Math|Vector|RotateVectorAroundAxis (Math|Rotator|GetRotationXVector (Math|Rotator|MakeRotfromZ (.Normal (Collision|BreakHitResult _outhit)))) _returnvalue_3 (.Normal (Collision|BreakHitResult _outhit))) (.Normal (Collision|BreakHitResult _outhit)))) (Math|Vector|MakeVector (Math|Random|RandomFloatinRangefromStream _returnvalue_2 (.ScaleMin_8_FE7EBF17444458437DDE60BE2A001AF5 (Utilities|Struct|BreakProceduralCollection _array_element_2)) (.ScaleMax_10_1A61A5DA412FCD256B46628068A60EA3 (Utilities|Struct|BreakProceduralCollection _array_element_2))) (Math|Random|RandomFloatinRangefromStream _returnvalue_2 (.ScaleMin_8_FE7EBF17444458437DDE60BE2A001AF5 (Utilities|Struct|BreakProceduralCollection _array_element_2)) (.ScaleMax_10_1A61A5DA412FCD256B46628068A60EA3 (Utilities|Struct|BreakProceduralCollection _array_element_2))) (Math|Random|RandomFloatinRangefromStream _returnvalue_2 (.ScaleMin_8_FE7EBF17444458437DDE60BE2A001AF5 (Utilities|Struct|BreakProceduralCollection _array_element_2)) (.ScaleMax_10_1A61A5DA412FCD256B46628068A60EA3 (Utilities|Struct|BreakProceduralCollection _array_element_2)))))) (Math|Vector|MakeVector 1.0 1.0 1.0)))))
            (Variables|/Script/Engine.Actor|SetTags (Utilities|Array|MakeArray (Variables|DoNotEdit|GetsearchTag) (Utilities|String|StringToName (Utilities|String|Append (Utilities|String|Append (Utilities|String|Append "PCOUNT_" (Utilities|String|ToString(Integer) _array_index)) "_") (Utilities|String|ToString(Integer) (- (Variables|Default|GetIteration) 1))))) _returnvalue_6))))
      (Utilities|Array|Clear (Variables|Default|GetActorEditLocationsArray)))))

(event Custom|1%20ScatterDetailAssets
  (bind _count (Variables|Default|GetCount))
  (bind _iteration (Variables|Default|GetIteration))
  (bind _randomseed (Variables|Procedural|GetRandomSeed))
  (bind _returnvalue (Math|Integer|int+int _randomseed _iteration _count))
  (bind _returnvalue_2 (Math|Random|MakeRandomStream _returnvalue))
  (bind _returnvalue_3 (Math|Random|RandomFloatinRangefromStream _returnvalue_2 0.0 3.6e+02))
  (bind _returnvalue_4 (Transformation|GetActorRotation))
  (Variables|Default|SetResetScatter false)
  (CallFunction|DeleteScatteredActors)
  (CallFunction|DeleteScatteredInstances)
  (Utilities|FlowControl|Sequence
    (:then_0
      ;[1 1 1] Make Unique Identifier
      (if (== (Variables|DoNotEdit|GetUniqueIdentifier) 0)
        (Variables|DoNotEdit|SetUniqueIdentifier (Math|Random|RandomIntegerinRange 1 999999999)))
      (bind _outactors (Actor|GetAllActorsOfClass (Utilities|GetClass self)))
      (for _array_element _outactors
        (if (Utilities|Equal(Object) _array_element self)
          _
          (elif (== (Variables|DoNotEdit|GetUniqueIdentifier _array_element) (Variables|DoNotEdit|GetUniqueIdentifier))
            (Variables|DoNotEdit|SetUniqueIdentifier (Math|Random|RandomIntegerinRange 1 999999999))))))
    (:then_1
      ;[1 1 1] makeUnique Tag
      (Variables|DoNotEdit|SetsearchTag (Utilities|String|StringToName (Utilities|String|Append "PPA_" (Utilities|String|ToString(Integer) (Variables|DoNotEdit|GetUniqueIdentifier))))))
    (:then_2
      (CallFunction|ScatterInstanceMeshes))
    (:then_3
      (for (_array_element_2 _array_index) (Variables|Default|GetACTOR_List)
        (Variables|Default|SetCount _array_index)
        (Math|Random|SetRandomStreamSeed (Variables|Default|GetStream) (+ (Variables|Procedural|GetRandomSeed) _array_index))
        (for _index (range 1 (+ (.Count_5_7BCF801C487720A69910A9A310D929B0 (Utilities|Struct|BreakProceduralCollection _array_element_2)) 1))
          (Variables|Default|SetIteration _index)
          (bind (_outhit _returnvalue_5) (Collision|LineTraceByChannel :Start (+ (Math|Vector|MakeVector 0.0 0.0 (Variables|Procedural|GetHeight)) (+ (Transformation|GetActorLocation) (Math|Vector|RotateVector (* (* (Math|Vector|Normalize (* (Math|Random|RandomUnitVectorfromStream (Variables|Default|GetStream)) (Math|Vector|MakeVector 1.0 1.0))) (Math|Random|RandomFloatinRangefromStream (Variables|Default|GetStream) (Variables|Procedural|GetRadius_Min) (Variables|Procedural|GetRadius_Max))) (Transformation|GetActorScale3D)) (Transformation|GetActorRotation)))) :End (+ (+ (Transformation|GetActorLocation) (Math|Vector|RotateVector (* (* (Math|Vector|Normalize (* (Math|Random|RandomUnitVectorfromStream (Variables|Default|GetStream)) (Math|Vector|MakeVector 1.0 1.0))) (Math|Random|RandomFloatinRangefromStream (Variables|Default|GetStream) (Variables|Procedural|GetRadius_Min) (Variables|Procedural|GetRadius_Max))) (Transformation|GetActorScale3D)) (Transformation|GetActorRotation))) (Math|Vector|MakeVector 0.0 0.0 -3e+03)) :TraceChannel "TraceTypeQuery10" :bTraceComplex true :DrawDebugType "ForOneFrame" :bIgnoreSelf true :TraceColor "(R=1.000000,G=0.000000,B=0.000000,A=1.000000)" :TraceHitColor "(R=0.000000,G=1.000000,B=0.000000,A=1.000000)" :DrawTime 2.0))
          (if _returnvalue_5
            (bind _returnvalue_6 (Game|SpawnActorfromClass "/Game/ArtTools/Procedural/Procedural_SupportActor.Procedural_SupportActor_C" (Math|Transform|MakeTransform (.Location (Collision|BreakHitResult _outhit)) (Math|Rotator|CombineRotators _returnvalue_4 (Math|Rotator|MakeRotfromXZ (Math|Vector|RotateVectorAroundAxis (Math|Rotator|GetRotationXVector (Math|Rotator|MakeRotfromZ (.Normal (Collision|BreakHitResult _outhit)))) _returnvalue_3 (.Normal (Collision|BreakHitResult _outhit))) (.Normal (Collision|BreakHitResult _outhit)))) (Math|Vector|MakeVector (Math|Random|RandomFloatinRangefromStream _returnvalue_2 (.ScaleMin_8_FE7EBF17444458437DDE60BE2A001AF5 (Utilities|Struct|BreakProceduralCollection _array_element_2)) (.ScaleMax_10_1A61A5DA412FCD256B46628068A60EA3 (Utilities|Struct|BreakProceduralCollection _array_element_2))) (Math|Random|RandomFloatinRangefromStream _returnvalue_2 (.ScaleMin_8_FE7EBF17444458437DDE60BE2A001AF5 (Utilities|Struct|BreakProceduralCollection _array_element_2)) (.ScaleMax_10_1A61A5DA412FCD256B46628068A60EA3 (Utilities|Struct|BreakProceduralCollection _array_element_2))) (Math|Random|RandomFloatinRangefromStream _returnvalue_2 (.ScaleMin_8_FE7EBF17444458437DDE60BE2A001AF5 (Utilities|Struct|BreakProceduralCollection _array_element_2)) (.ScaleMax_10_1A61A5DA412FCD256B46628068A60EA3 (Utilities|Struct|BreakProceduralCollection _array_element_2))))) "Undefined" "OverrideRootScale" 0 0))
            (Class|ProceduralSupportActor|SetChildActor _returnvalue_6 (.Blueprint_24_336B25D4459493F1D17B81BA5C9C10CA (Utilities|Struct|BreakProceduralCollection _array_element_2)) (Math|Transform|MakeRelativeTransform (Math|Transform|MakeTransform (.Location (Collision|BreakHitResult _outhit)) (Math|Rotator|CombineRotators _returnvalue_4 (Math|Rotator|MakeRotfromXZ (Math|Vector|RotateVectorAroundAxis (Math|Rotator|GetRotationXVector (Math|Rotator|MakeRotfromZ (.Normal (Collision|BreakHitResult _outhit)))) _returnvalue_3 (.Normal (Collision|BreakHitResult _outhit))) (.Normal (Collision|BreakHitResult _outhit)))) (Math|Vector|MakeVector (Math|Random|RandomFloatinRangefromStream _returnvalue_2 (.ScaleMin_8_FE7EBF17444458437DDE60BE2A001AF5 (Utilities|Struct|BreakProceduralCollection _array_element_2)) (.ScaleMax_10_1A61A5DA412FCD256B46628068A60EA3 (Utilities|Struct|BreakProceduralCollection _array_element_2))) (Math|Random|RandomFloatinRangefromStream _returnvalue_2 (.ScaleMin_8_FE7EBF17444458437DDE60BE2A001AF5 (Utilities|Struct|BreakProceduralCollection _array_element_2)) (.ScaleMax_10_1A61A5DA412FCD256B46628068A60EA3 (Utilities|Struct|BreakProceduralCollection _array_element_2))) (Math|Random|RandomFloatinRangefromStream _returnvalue_2 (.ScaleMin_8_FE7EBF17444458437DDE60BE2A001AF5 (Utilities|Struct|BreakProceduralCollection _array_element_2)) (.ScaleMax_10_1A61A5DA412FCD256B46628068A60EA3 (Utilities|Struct|BreakProceduralCollection _array_element_2))))) (Transformation|GetActorTransform)) (Math|Transform|MakeTransform (.Location (Collision|BreakHitResult _outhit)) (Math|Rotator|CombineRotators _returnvalue_4 (Math|Rotator|MakeRotfromXZ (Math|Vector|RotateVectorAroundAxis (Math|Rotator|GetRotationXVector (Math|Rotator|MakeRotfromZ (.Normal (Collision|BreakHitResult _outhit)))) _returnvalue_3 (.Normal (Collision|BreakHitResult _outhit))) (.Normal (Collision|BreakHitResult _outhit)))) (Math|Vector|MakeVector (Math|Random|RandomFloatinRangefromStream _returnvalue_2 (.ScaleMin_8_FE7EBF17444458437DDE60BE2A001AF5 (Utilities|Struct|BreakProceduralCollection _array_element_2)) (.ScaleMax_10_1A61A5DA412FCD256B46628068A60EA3 (Utilities|Struct|BreakProceduralCollection _array_element_2))) (Math|Random|RandomFloatinRangefromStream _returnvalue_2 (.ScaleMin_8_FE7EBF17444458437DDE60BE2A001AF5 (Utilities|Struct|BreakProceduralCollection _array_element_2)) (.ScaleMax_10_1A61A5DA412FCD256B46628068A60EA3 (Utilities|Struct|BreakProceduralCollection _array_element_2))) (Math|Random|RandomFloatinRangefromStream _returnvalue_2 (.ScaleMin_8_FE7EBF17444458437DDE60BE2A001AF5 (Utilities|Struct|BreakProceduralCollection _array_element_2)) (.ScaleMax_10_1A61A5DA412FCD256B46628068A60EA3 (Utilities|Struct|BreakProceduralCollection _array_element_2))))) "None" :!class /Game/ArtTools/Procedural/Procedural_SupportActor.Procedural_SupportActor_C)
            (if (not (Variables|Default|GetResetScatter))
              (Variables|Default|SetCurrentEditLocation (Math|Transform|MakeTransform "0, 0, 0"))
              (bind _array_element_3 (Utilities|Array|ForEachLoopwithBreak (Variables|Default|GetActorEditLocationsArray))
                (:LoopBody
                  (if (== (Utilities|Struct|BreakProceduralEditLocation _array_element_3) (Variables|Default|GetCount))
                    (if (== (.Iteration_4_6A5FC8554676355A6CD0528E37AA79EB (Utilities|Struct|BreakProceduralEditLocation _array_element_3)) (- (Variables|Default|GetIteration) 1))
                      (Variables|Default|SetCurrentEditLocation (.EditTransform_9_49D70F1F48C2C776CD90B5BC2FE0BEAE (Utilities|Struct|BreakProceduralEditLocation _array_element_3)))
                      (Variables|Default|SetResetScatter false))))
                (:Completed
                  _))
              (Class|ProceduralSupportActor|OverrideEditTransform _returnvalue_6 (Math|Transform|MakeRelativeTransform (Variables|Default|GetCurrentEditLocation) (Math|Transform|MakeTransform (.location (Math|Transform|MakeTransform (.Location (Collision|BreakHitResult _outhit)) (Math|Rotator|CombineRotators _returnvalue_4 (Math|Rotator|MakeRotfromXZ (Math|Vector|RotateVectorAroundAxis (Math|Rotator|GetRotationXVector (Math|Rotator|MakeRotfromZ (.Normal (Collision|BreakHitResult _outhit)))) _returnvalue_3 (.Normal (Collision|BreakHitResult _outhit))) (.Normal (Collision|BreakHitResult _outhit)))) (Math|Vector|MakeVector (Math|Random|RandomFloatinRangefromStream _returnvalue_2 (.ScaleMin_8_FE7EBF17444458437DDE60BE2A001AF5 (Utilities|Struct|BreakProceduralCollection _array_element_2)) (.ScaleMax_10_1A61A5DA412FCD256B46628068A60EA3 (Utilities|Struct|BreakProceduralCollection _array_element_2))) (Math|Random|RandomFloatinRangefromStream _returnvalue_2 (.ScaleMin_8_FE7EBF17444458437DDE60BE2A001AF5 (Utilities|Struct|BreakProceduralCollection _array_element_2)) (.ScaleMax_10_1A61A5DA412FCD256B46628068A60EA3 (Utilities|Struct|BreakProceduralCollection _array_element_2))) (Math|Random|RandomFloatinRangefromStream _returnvalue_2 (.ScaleMin_8_FE7EBF17444458437DDE60BE2A001AF5 (Utilities|Struct|BreakProceduralCollection _array_element_2)) (.ScaleMax_10_1A61A5DA412FCD256B46628068A60EA3 (Utilities|Struct|BreakProceduralCollection _array_element_2)))))) (.rotation (Math|Transform|MakeTransform (.Location (Collision|BreakHitResult _outhit)) (Math|Rotator|CombineRotators _returnvalue_4 (Math|Rotator|MakeRotfromXZ (Math|Vector|RotateVectorAroundAxis (Math|Rotator|GetRotationXVector (Math|Rotator|MakeRotfromZ (.Normal (Collision|BreakHitResult _outhit)))) _returnvalue_3 (.Normal (Collision|BreakHitResult _outhit))) (.Normal (Collision|BreakHitResult _outhit)))) (Math|Vector|MakeVector (Math|Random|RandomFloatinRangefromStream _returnvalue_2 (.ScaleMin_8_FE7EBF17444458437DDE60BE2A001AF5 (Utilities|Struct|BreakProceduralCollection _array_element_2)) (.ScaleMax_10_1A61A5DA412FCD256B46628068A60EA3 (Utilities|Struct|BreakProceduralCollection _array_element_2))) (Math|Random|RandomFloatinRangefromStream _returnvalue_2 (.ScaleMin_8_FE7EBF17444458437DDE60BE2A001AF5 (Utilities|Struct|BreakProceduralCollection _array_element_2)) (.ScaleMax_10_1A61A5DA412FCD256B46628068A60EA3 (Utilities|Struct|BreakProceduralCollection _array_element_2))) (Math|Random|RandomFloatinRangefromStream _returnvalue_2 (.ScaleMin_8_FE7EBF17444458437DDE60BE2A001AF5 (Utilities|Struct|BreakProceduralCollection _array_element_2)) (.ScaleMax_10_1A61A5DA412FCD256B46628068A60EA3 (Utilities|Struct|BreakProceduralCollection _array_element_2)))))) (Math|Vector|MakeVector 1.0 1.0 1.0)))))
            (Variables|/Script/Engine.Actor|SetTags (Utilities|Array|MakeArray (Variables|DoNotEdit|GetsearchTag) (Utilities|String|StringToName (Utilities|String|Append (Utilities|String|Append (Utilities|String|Append "PCOUNT_" (Utilities|String|ToString(Integer) _array_index)) "_") (Utilities|String|ToString(Integer) (- (Variables|Default|GetIteration) 1))))) _returnvalue_6))))
      (Utilities|Array|Clear (Variables|Default|GetActorEditLocationsArray)))))

(event Custom|0%20Align%20to%20floor
  (CallFunction|ActorAligntoFloor))

(fn ActorAligntoFloor
  (bind _outactors (Actor|GetAllActorswithTag (Variables|DoNotEdit|GetsearchTag)))
  (for _array_element _outactors
    (bind _outhits (Collision|MultiLineTraceForObjects :Start (+ (Transformation|GetActorLocation _array_element) (Math|Vector|MakeVector 0.0 0.0 (Variables|Procedural|GetHeight))) :End (- (Transformation|GetActorLocation _array_element) (* (Math|Vector|MakeVector 0.0 0.0 (Variables|Procedural|GetHeight)) 1e+01)) :ObjectTypes (Utilities|Array|MakeArray "ObjectTypeQuery1") :bTraceComplex true :DrawDebugType "ForOneFrame" :bIgnoreSelf true :TraceColor "(R=1.000000,G=0.000000,B=0.000000,A=1.000000)" :TraceHitColor "(R=0.000000,G=1.000000,B=0.000000,A=1.000000)" :DrawTime 5.0))
    (Development|PrintString (Utilities|String|ToString(Integer) (Utilities|Array|Length _outhits)))
    (bind _array_element_2 (Utilities|Array|ForEachLoopwithBreak _outhits)
      (:LoopBody
        (Utilities|Casting|CastToLandscape (.HitActor (Collision|BreakHitResult _array_element_2)))
        (Transformation|SetActorLocation _array_element (.ImpactPoint (Collision|BreakHitResult _array_element_2)) false true))
      (:Completed
        _))
    (bind _asprocedural_support_actor (Utilities|Casting|CastToProcedural_SupportActor _array_element))
    (Class|ProceduralSupportActor|SetEditTransform _asprocedural_support_actor)))

(fn DeleteScatteredActors
  (bind _outactors (Actor|GetAllActorswithTag (Variables|DoNotEdit|GetsearchTag)))
  (for _array_element _outactors
    (bind _asprocedural_support_actor (Utilities|Casting|CastToProcedural_SupportActor _array_element))
    ;[1 1 1] Needs two dimensoanl tag
    (for _array_element_2 (Variables|/Script/Engine.Actor|GetTags _asprocedural_support_actor)
      (if (Utilities|String|Contains (Utilities|String|ToString(Name) _array_element_2) "PCOUNT_")
        (Utilities|Array|Add (Variables|Default|GetActorEditLocationsArray) (Utilities|Struct|MakeProceduralEditLocation (Utilities|String|StringToInteger (Utilities|String|GetSubstring (Utilities|String|ToString(Name) _array_element_2) 7)) (Utilities|String|StringToInteger (Utilities|String|GetSubstring (Utilities|String|ToString(Name) _array_element_2) 9 (- (Utilities|String|Len (Utilities|String|ToString(Name) _array_element_2)) 9))) (Transformation|GetActorTransform _asprocedural_support_actor)))
        (Actor|DestroyActor _array_element)))))

(fn DeleteScatteredInstances
  (for _array_element (Variables|Default|GetHISM_List)
    (Components|InstancedStaticMesh|ClearInstances _array_element)))

(fn ScatterInstanceMeshes
  (bind _data_meshlist (Variables|Default|GetDATA_MeshList))
  (for (_array_element _array_index) (Variables|Default|GetHISM_List)
    (Math|Random|SetRandomStreamSeed (Variables|Default|GetStream) (+ (Variables|Procedural|GetRandomSeed) _array_index))
    (for _index (range 1 (+ (.Count_5_7BCF801C487720A69910A9A310D929B0 (Utilities|Struct|BreakProceduralCollection (Utilities|Array|Get(aref) _data_meshlist _array_index))) 1))
      (Variables|Default|SetIteration _index)
      (bind _outactors (Actor|GetAllActorswithTag (Variables|DoNotEdit|GetsearchTag)))
      (bind _outhits (Collision|MultiLineTraceForObjects (+ (Math|Vector|MakeVector 0.0 0.0 (Variables|Procedural|GetHeight)) (+ (Transformation|GetActorLocation) (Math|Vector|RotateVector (* (* (Math|Vector|Normalize (* (Math|Random|RandomUnitVectorfromStream (Variables|Default|GetStream)) (Math|Vector|MakeVector 1.0 1.0))) (Math|Random|RandomFloatinRangefromStream (Variables|Default|GetStream) (Variables|Procedural|GetRadius_Min) (Variables|Procedural|GetRadius_Max))) (Transformation|GetActorScale3D)) (Transformation|GetActorRotation)))) (+ (+ (Transformation|GetActorLocation) (Math|Vector|RotateVector (* (* (Math|Vector|Normalize (* (Math|Random|RandomUnitVectorfromStream (Variables|Default|GetStream)) (Math|Vector|MakeVector 1.0 1.0))) (Math|Random|RandomFloatinRangefromStream (Variables|Default|GetStream) (Variables|Procedural|GetRadius_Min) (Variables|Procedural|GetRadius_Max))) (Transformation|GetActorScale3D)) (Transformation|GetActorRotation))) (Math|Vector|MakeVector 0.0 0.0 -3e+03)) (Utilities|Array|MakeArray "ObjectTypeQuery1") true _outactors))
      (bind _array_element_2 (Utilities|Array|ForEachLoopwithBreak _outhits)
        (:LoopBody
          (Utilities|Casting|CastToLandscape (.HitActor (Collision|BreakHitResult _array_element_2)))
          (Components|InstancedStaticMesh|AddInstance _array_element (Math|Transform|MakeTransform (.Location (Collision|BreakHitResult _array_element_2)) (Math|Rotator|MakeRotfromXZ (Math|Vector|RotateVectorAroundAxis (Math|Rotator|GetRotationXVector (Math|Rotator|MakeRotfromZ (.Normal (Collision|BreakHitResult _array_element_2)))) (Math|Random|RandomFloatinRangefromStream (Math|Random|MakeRandomStream (+ (Variables|Procedural|GetRandomSeed) (Variables|Default|GetIteration))) 0.0 3.6e+02) (.Normal (Collision|BreakHitResult _array_element_2))) (.Normal (Collision|BreakHitResult _array_element_2))) (Math|Vector|MakeVector (Math|Random|RandomFloatinRangefromStream (Math|Random|MakeRandomStream (+ (Variables|Procedural|GetRandomSeed) (Variables|Default|GetIteration))) (.ScaleMin_8_FE7EBF17444458437DDE60BE2A001AF5 (Utilities|Struct|BreakProceduralCollection (Utilities|Array|Get(aref) _data_meshlist _array_index))) (.ScaleMax_10_1A61A5DA412FCD256B46628068A60EA3 (Utilities|Struct|BreakProceduralCollection (Utilities|Array|Get(aref) _data_meshlist _array_index)))) (Math|Random|RandomFloatinRangefromStream (Math|Random|MakeRandomStream (+ (Variables|Procedural|GetRandomSeed) (Variables|Default|GetIteration))) (.ScaleMin_8_FE7EBF17444458437DDE60BE2A001AF5 (Utilities|Struct|BreakProceduralCollection (Utilities|Array|Get(aref) _data_meshlist _array_index))) (.ScaleMax_10_1A61A5DA412FCD256B46628068A60EA3 (Utilities|Struct|BreakProceduralCollection (Utilities|Array|Get(aref) _data_meshlist _array_index)))) (Math|Random|RandomFloatinRangefromStream (Math|Random|MakeRandomStream (+ (Variables|Procedural|GetRandomSeed) (Variables|Default|GetIteration))) (.ScaleMin_8_FE7EBF17444458437DDE60BE2A001AF5 (Utilities|Struct|BreakProceduralCollection (Utilities|Array|Get(aref) _data_meshlist _array_index))) (.ScaleMax_10_1A61A5DA412FCD256B46628068A60EA3 (Utilities|Struct|BreakProceduralCollection (Utilities|Array|Get(aref) _data_meshlist _array_index)))))) true))))))

(fn UpdateSupportActors
  (bind _searchtag (Variables|DoNotEdit|GetsearchTag))
  (Development|PrintString (Utilities|String|ToString(Name) _searchtag))
  (bind _outactors (Actor|GetAllActorswithTag _searchtag))
  (for _array_element _outactors
    (bind _asprocedural_support_actor (Utilities|Casting|CastToProcedural_SupportActor _array_element))
    (Class|ProceduralSupportActor|UpdateTransform _asprocedural_support_actor (Transformation|GetActorTransform))))

(fn ConstructionScript
  (bind _hierarchicalinstancedstaticmesh (Variables|Default|GetHierarchicalInstancedStaticMesh))
  (bind _hierarchicalinstancedstaticmesh1 (Variables|Default|GetHierarchicalInstancedStaticMesh1))
  (bind _hierarchicalinstancedstaticmesh2 (Variables|Default|GetHierarchicalInstancedStaticMesh2))
  (bind _hierarchicalinstancedstaticmesh3 (Variables|Default|GetHierarchicalInstancedStaticMesh3))
  (bind _hierarchicalinstancedstaticmesh4 (Variables|Default|GetHierarchicalInstancedStaticMesh4))
  (bind _hierarchicalinstancedstaticmesh5 (Variables|Default|GetHierarchicalInstancedStaticMesh5))
  (Utilities|Array|Clear (Variables|Default|GetACTOR_List))
  (Utilities|Array|Clear (Variables|Default|GetDATA_MeshList))
  (Utilities|FlowControl|Sequence
    (:then_0
      (for _array_element (Variables|/Game/ArtTools/Procedural/Procedural_DataAsset.Procedural_DataAsset_C|GetProcerualCollection (Variables|Procedural|GetProceduralDataCollection))
        (if (Utilities|IsValid (Utilities|Struct|BreakProceduralCollection _array_element) :!class /Script/Engine.KismetSystemLibrary)
          (Utilities|Array|Add (Variables|Default|GetDATA_MeshList) _array_element)
          (elif (Utilities|IsValidClass (.Blueprint_24_336B25D4459493F1D17B81BA5C9C10CA (Utilities|Struct|BreakProceduralCollection _array_element)))
            (Utilities|Array|Add (Variables|Default|GetACTOR_List) _array_element)))))
    (:then_1
      (Utilities|Array|Clear (Variables|Default|GetHISM_List))
      (Utilities|FlowControl|Sequence
        (:then_0
          (if (< (Utilities|Array|Length (Variables|Default|GetDATA_MeshList)) 1)
            (Components|DestroyComponent _hierarchicalinstancedstaticmesh)
            (else
              (Utilities|Array|Add (Variables|Default|GetHISM_List) _hierarchicalinstancedstaticmesh))))
        (:then_1
          (if (< (Utilities|Array|Length (Variables|Default|GetDATA_MeshList)) 2)
            (Components|DestroyComponent _hierarchicalinstancedstaticmesh1)
            (else
              (Utilities|Array|Add (Variables|Default|GetHISM_List) _hierarchicalinstancedstaticmesh1))))
        (:then_2
          (if (< (Utilities|Array|Length (Variables|Default|GetDATA_MeshList)) 3)
            (Components|DestroyComponent _hierarchicalinstancedstaticmesh2)
            (else
              (Utilities|Array|Add (Variables|Default|GetHISM_List) _hierarchicalinstancedstaticmesh2))))
        (:then_3
          (if (< (Utilities|Array|Length (Variables|Default|GetDATA_MeshList)) 4)
            (Components|DestroyComponent _hierarchicalinstancedstaticmesh3)
            (else
              (Utilities|Array|Add (Variables|Default|GetHISM_List) _hierarchicalinstancedstaticmesh3))))
        (:then_4
          (if (< (Utilities|Array|Length (Variables|Default|GetDATA_MeshList)) 5)
            (Components|DestroyComponent _hierarchicalinstancedstaticmesh4)
            (else
              (Utilities|Array|Add (Variables|Default|GetHISM_List) _hierarchicalinstancedstaticmesh4))))
        (:then_5
          (if (< (Utilities|Array|Length (Variables|Default|GetDATA_MeshList)) 6)
            (Components|DestroyComponent _hierarchicalinstancedstaticmesh5)
            (else
              (Utilities|Array|Add (Variables|Default|GetHISM_List) _hierarchicalinstancedstaticmesh5))))))
    (:then_2
      (for (_array_element_2 _array_index) (Variables|Default|GetDATA_MeshList)
        (bind _asstatic_mesh (Utilities|Casting|CastToStaticMesh (Utilities|Struct|BreakProceduralCollection _array_element_2)))
        (Components|StaticMesh|SetStaticMesh (Utilities|Array|Get(acopy) (Variables|Default|GetHISM_List) _array_index) _asstatic_mesh)))
    (:then_3
      (CallFunction|UpdateSupportActors))
    (:then_4
      (CallFunction|DeleteScatteredInstances)
      (CallFunction|ScatterInstanceMeshes))))
